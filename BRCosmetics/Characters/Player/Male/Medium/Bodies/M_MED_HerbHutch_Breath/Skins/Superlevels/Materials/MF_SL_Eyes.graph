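# Material function for the eyes of the HerbHutch "Superlevels" skin, setting black base color, a normal and glowing emissive on material attributes.
# An FNEye-style refraction chain derives eye UVs from cornea IOR, a displacement-map iris depth and left/right UV positions.
# The glow lerps Eye Color to a lightened tint through an iris sphere mask, fading with inverse Fresnel.
# A separate EyeMask output combines two UV sphere masks; several color nodes are unconnected.
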MaterialFunction /BRCosmetics/Characters/Player/Male/Medium/Bodies/M_MED_HerbHutch_Breath/Skins/Superlevels/Materials/MF_SL_Eyes
graph {nodes [
  n1 = ScalarParameter() {defaultValue=0.02, group=Eye Shader, parameterName=Eye Refraction Mult, sortPriority=25}
  n2 = Constant2Vector() {g=1, r=-1}
  n3 = NamedRerouteUsage()
  n4 = MaterialFunctionCall() {materialFunction=&TangentBasis}
  n5 = NamedRerouteUsage()
  n6 = DotProduct(A=@n4.X U, B=@n5)
  n7 = Normalize(VectorInput=Subtract(A=@n4.X U, B=Multiply(A=@n6, B=@n5)))
  n8 = DotProduct(A=@n7, B=@n3)
  n9 = DotProduct(A=CrossProduct(A=@n7, B=@n5), B=@n3)
  n10 = Multiply(
   A=Multiply(
    A=Multiply(A=@n1, B=@n2)
    B=MaterialFunctionCall(X=@n8, Y=@n9) {materialFunction=&MakeFloat2}.Result
   )
   B=NamedRerouteUsage()
  )
  Add(A=Add(A=NamedRerouteUsage(), B=@n10), B=@n10)
  n11 = Fresnel() {baseReflectFraction=0, exponent=1}
  n12 = VectorParameter() {group=SL_S31_Eyes, parameterName=Eye Color, defaultValue=<rgba 0.3185 0.7083 0.0133 1>}
  n13 = Saturate(None=LinearInterpolate(A=@n12.RGB) {constAlpha=0.35})
  n14 = SphereMask(
   A=NamedRerouteUsage()
   B=Constant() {r=0.5}
   Radius=ScalarParameter() {defaultValue=0.35, group=SL_S31_Eyes, parameterName=EyeTintMask_Radius, sliderMax=1}
  ) {attenuationRadius=0.4, hardnessPercent=80}
  n15 = Multiply(A=@n14, B=ScalarParameter() {defaultValue=1, group=SL_S31_Eyes, parameterName=EyeIrisIntensity})
  n16 = Multiply(
   A=Multiply(A=OneMinus(None=@n11), B=LinearInterpolate(A=@n12.RGB, B=@n13, Alpha=@n15))
   B=ScalarParameter() {defaultValue=1.5, parameterName=Eye Color Intensity}
  )
  FunctionOutput(None=SetMaterialAttributes!(MaterialAttributes, BaseColor=Constant(), Normal=NamedRerouteUsage(), EmissiveColor=@n16) {attributeSetTypes [#1, #2, #3]})
  n17 = TextureCoordinate()
  n18 = ScalarParameter() {defaultValue=0.08, group=SL_S31_Eyes, parameterName=EyeMask_Radius, sliderMax=1}
  n19 = SphereMask(A=@n17, B=NamedRerouteUsage(), Radius=@n18) {attenuationRadius=0.4, hardnessPercent=80}
  n20 = SphereMask(A=@n17, B=NamedRerouteUsage(), Radius=@n18) {attenuationRadius=0.4, hardnessPercent=80}
  FunctionOutput(None=Saturate(None=Add(A=@n19, B=@n20))) {outputName=EyeMask, sortPriority=1}
  n21 = Desaturation(None=NamedRerouteUsage(), Fraction=Constant() {r=1})
  n22 = Multiply(A=SmoothStep(Value=@n21) {constMax=0.1}, B=@n16)
  LinearInterpolate(A=@n21, B=@n22, Alpha=@n14)
  Multiply(A=@n22, B=@n14)
  NamedRerouteDeclaration(None=Add(A=@n10, B=NamedRerouteUsage())) {name=Refracted UVs, nodeColor=<rgba 0.6406 1 0 1>}
  n23 = TextureCoordinate()
  n24 = NamedRerouteDeclaration(None=Subtract(A=NamedRerouteUsage(), B=NamedRerouteUsage())) {name=LeftEyeUVDelta, nodeColor=<rgba 0.9688 1 0 1>}
  NamedRerouteDeclaration(None=LinearInterpolate(A=@n23, B=Subtract(A=@n23, B=@n24), Alpha=NamedRerouteUsage())) {name=UVs, nodeColor=<rgba 0 1 0.2969 1>}
  n25 = NamedRerouteUsage()
  n26 = Constant2Vector() {g=0.5, r=0.5}
  n27 = NamedRerouteUsage()
  n28 = NamedRerouteUsage()
  NamedRerouteDeclaration(None=Power(Base=SphereMask(A=@n25, B=@n26, Radius=@n27, Hardness=@n28) {attenuationRadius=0.365, hardnessPercent=50}) {constExponent=0.5}) {name=eye displacement mask, nodeColor=<rgba 0.875 1 0 1>}
  n29 = NamedRerouteUsage()
  n30 = DotProduct(A=CameraVectorWS(), B=NamedRerouteUsage())
  n31 = LinearInterpolate(Alpha=Multiply(A=@n30, B=@n30)) {constA=0.325}
  NamedRerouteDeclaration(None=Multiply(A=NamedRerouteUsage(), B=Divide(A=@n29, B=@n31) {desc=heightW / cosAlpha - scale distance to iris plane by facing ratio of camera and plane})) {name=Refraction Vector, nodeColor=<rgba 0.0781 1 0 1>}
  n32 = NamedRerouteDeclaration(None=LinearInterpolate(
    A=NamedRerouteUsage()
    B=NamedRerouteUsage()
    Alpha=ScalarParameter() {defaultValue=1, group=Eye Shader, parameterName=Eye Refraction Mix, sliderMax=1, sortPriority=27}
   )) {name="Eye UVs (normalized)", nodeColor=<rgba 0 1 0.5232 1>}
  n33 = NamedRerouteUsage()
  NamedRerouteDeclaration(None=LinearInterpolate(A=@n32, B=Add(A=@n32, B=@n33), Alpha=NamedRerouteUsage())) {name=EyeUVs, nodeColor=<rgba 0 1 0.6016 1>}
  NamedRerouteDeclaration(None=ScalarParameter() {defaultValue=0.078, group=SL_S31_Eyes, parameterName="Eye Cornea Radius (UV)", sortPriority=10}) {name=EyeCorneaRadius, nodeColor=<rgba 0 1 0.2266 1>}
  NamedRerouteDeclaration(None=PixelNormalWS()) {name=_Normal, nodeColor=<rgba 0 0 1 1>}
  NamedRerouteDeclaration(None=Divide(A=Constant() {desc=airIoR, r=1.0003}, B=ScalarParameter() {defaultValue=1.336, group=Eye Shader, parameterName=Eye Cornea IOR}) {constA=1.0003}) {name=n, nodeColor=<rgba 0 1 0.2031 1>}
  n34 = NamedRerouteUsage()
  n35 = NamedRerouteUsage()
  NamedRerouteDeclaration(None=Multiply(A=NamedRerouteUsage(), B=DotProduct(A=@n34, B=@n35) {desc=facing}) {desc=w}) {name=w, nodeColor=<rgba 1 0.8297 0 1>}
  n36 = NamedRerouteUsage()
  n37 = NamedRerouteUsage()
  n38 = Multiply(A=Subtract(A=@n36, B=@n37), B=Add(A=@n36, B=@n37))
  NamedRerouteDeclaration(None=SquareRoot(None=Add(A=@n38)) {desc=k}) {name=k, nodeColor=<rgba 0.518 1 0 1>}
  NamedRerouteDeclaration(None=Transform(None=NamedRerouteUsage())) {name=EyeDirection, nodeColor=<rgba 1 0.3047 0 1>}
  n39 = NamedRerouteUsage()
  n40 = NamedRerouteUsage()
  n41 = NamedRerouteUsage()
  n42 = Subtract(A=Divide(A=Multiply(A=@n39, B=@n40), B=@n41)) {constB=-0.5}
  n43 = NamedRerouteUsage()
  n44 = NamedRerouteUsage()
  n45 = Divide(A=Multiply(A=Multiply(A=@n43, B=@n44)) {constB=-1, desc=compensates for normal map not being 0-1}, B=@n41)
  NamedRerouteDeclaration(None=Saturate(None=Add(A=@n42, B=@n45))) {name=Iris UVs, nodeColor=<rgba 0.2422 1 0 1>}
  NamedRerouteDeclaration(None=Constant() {r=0.365}) {name=IrisSize, nodeColor=<rgba 1 0 0.7969 1>}
  n46 = Multiply(A=NamedRerouteUsage(), B=NamedRerouteUsage())
  n47 = NamedRerouteUsage()
  n48 = Multiply(A=NamedRerouteUsage(), B=NamedRerouteUsage())
  n49 = Add(
   A=Subtract(A=Divide(A=@n46, B=@n47)) {constB=-0.5}
   B=Divide(A=Multiply(A=@n48) {constB=-1, desc=compensates for normal map not being 0-1}, B=@n47)
  )
  n50 = TextureSampleParameter2D(UVs=NamedRerouteDeclaration(None=Saturate(None=@n49)) {name=Iris UVs_1, nodeColor=<rgba 0.2422 1 0 1>}) {
   group=Eye Shader
   parameterName=Eye Iris Normal
   samplerType=SAMPLERTYPE_Normal
   sortPriority=20
   texture=/Game/Characters/Shared/FNEye/Textures/T_Eye_Sphere_N
  }
  NamedRerouteDeclaration(None=@n50.RGB) {name="eyeNormal (unflattened)", nodeColor=<rgba 0 1 0.1797 1>}
  NamedRerouteDeclaration(None=ScalarParameter() {defaultValue=0.2, group=SL_S31_Eyes, parameterName=Eye Cornea Mask Hardness, sortPriority=11}) {name=Eye Cornea Mask Hardness Reroute, nodeColor=<rgba 0.8047 0 1 1>}
  n51 = TextureSample(UVs=NamedRerouteUsage()) {
   constCoordinate=1
   desc=UVs not scaled by center as the base UVs are a ortho projection and this depth needs to stay in that space
   samplerType=SAMPLERTYPE_LinearColor
   texture=/Game/Characters/Shared/FNEye/Textures/T_EyeMidPlaneDisplacement
  }
  n52 = ScalarParameter() {defaultValue=0.4, group=Eye Shader, parameterName=Eye Iris UV Radius, sliderMax=0.5}
  n53 = TextureSample(UVs=MaterialFunctionCall(X=Add(A=@n52) {constB=0.5}, Y=Constant() {r=0.5}) {materialFunction=&MakeFloat2}.Result) {
   desc="Sample Displacement map as above, sampled at the Iris Radius"
   samplerType=SAMPLERTYPE_LinearColor
   texture=&T_EyeMidPlaneDisplacement
  }
  n54 = Subtract(A=@n51.RGB, B=@n53.R)
  n55 = Constant()
  NamedRerouteDeclaration(None=Multiply(A=Max(A=@n54, B=@n55), B=ScalarParameter() {defaultValue=0.5, group=Eye Shader, parameterName=Eye Iris Depth Scale}) {desc=Iris Depth}) {name=height, nodeColor=<rgba 0 1 0.4375 1>}
  n56 = NamedRerouteUsage()
  n57 = NamedRerouteUsage()
  n58 = TextureCoordinate()
  n59 = NamedRerouteUsage()
  NamedRerouteDeclaration(None=SphereMask(A=Multiply(A=@n56, B=@n57), B=Multiply(A=@n58, B=@n57) {constB=2}, Radius=Multiply(A=@n59) {constB=0.5}) {attenuationRadius=0.1}) {name=LeftEyeMask, nodeColor=<rgba 1 0 0.5391 1>}
  n60 = VectorParameter() {
   desc="uvPosition (UV0) of the left Eye. Top left is 0,0\r\nSet it to the same as the right eye if the same texture is used for both"
   group=SL_S31_Eyes
   parameterName="Eye Left UV Position (UV0)"
   sortPriority=13
   defaultValue=<rgba 0.94 0.35 0 1>
  }
  NamedRerouteDeclaration(None=ComponentMask(None=@n60.RGB) {g=true, r=true}) {name=Left Eye UVPosition, nodeColor=<rgba 1 0.375 0 1>}
  n61 = ScalarParameter() {defaultValue=2, desc="Texture Width/Height (2048x1024 would be 2)", group=SL_S31_Eyes, parameterName=Eye Texture AspectRatio, sortPriority=1}
  NamedRerouteDeclaration(None=MaterialFunctionCall(X=Constant() {r=1}, Y=Divide(B=@n61)) {materialFunction=&MakeFloat2}.Result) {name=AspectRatio, nodeColor=<rgba 1 0.4688 0 1>}
  NamedRerouteDeclaration(None=MaterialFunctionCall(BaseNormal=NamedRerouteUsage(), AdditionalNormal=NamedRerouteUsage()) {materialFunction=&BlendAngleCorrectedNormals}.Result) {name=EyeNormal, nodeColor=<rgba 0 1 0.9062 1>}
  n62 = TextureSampleParameter2D() {
   group=Eye Shader
   parameterName=Normal
   samplerType=SAMPLERTYPE_Normal
   sortPriority=0
   texture=&FlatNormalBlue
  }
  NamedRerouteDeclaration(None=@n62.RGB) {name=Normal, nodeColor=<rgba 0.1435 0.3013 1 1>}
  n63 = VectorParameter() {group=SL_S31_Eyes, parameterName="Eye Right UV Position (UV0)", sortPriority=12, defaultValue=<rgba 0.94 0.122 0 0>}
  NamedRerouteDeclaration(None=ComponentMask(None=@n63.RGB) {g=true, r=true}) {name=RightEyeUVPosition, nodeColor=<rgba 0.7812 1 0 1>}
  NamedRerouteDeclaration(None=TextureSampleParameter2D(UVs=NamedRerouteUsage()) {group=Eye Shader, parameterName=Diffuse, sortPriority=0, texture=/Game/Characters/Shared/FNEye/Textures/T_Eye_D}.RGB) {name=DiffuseMap, nodeColor=<rgba 0 0.8744 1 1>}
  NamedRerouteDeclaration(None=CameraVectorWS()) {name=Camera Vector, nodeColor=<rgba 1 0 1 1>}
  n64 = NamedRerouteUsage()
  n65 = NamedRerouteUsage()
  n66 = Subtract(A=Multiply(A=Subtract(A=@n64, B=@n65), B=NamedRerouteUsage()), B=Multiply(A=NamedRerouteUsage(), B=NamedRerouteUsage())) {desc=t}
  NamedRerouteDeclaration(None=Multiply(A=Normalize(VectorInput=@n66)) {constB=-1}) {name=Eye Refraction Direction, nodeColor=<rgba 0 0.2882 1 1>}
  NamedRerouteDeclaration(None=MaterialFunctionCall(Normal=@n50.RGB, Flatness=ScalarParameter() {defaultValue=0.25, group=Eye Shader, parameterName=Eye Iris Normal Flatten, sortPriority=21}) {materialFunction=&FlattenNormal}.Result) {name="eyeNormal (flattened)", nodeColor=<rgba 0.4766 1 0 1>}
  VectorParameter() {group=SL_S31_Eyes, parameterName=Eye Color 2, defaultValue=<rgba 0.7189 0.9387 0.4889 1>}
  VectorParameter() {group=Eye Color Tinting, parameterName=EyeTintColor, defaultValue=<rgba 5.9817 10 5.4688 1>}
  VertexNormalWS()
 ]}
