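# Material function for the Season 31 Superlevels skin effect on the HerbHutch_Breath character, restyling incoming material attributes within FX-mask regions.
# It builds emissive and surface layers from tiered panning noise, scrolling gradients, fresnel edges and thin-film tint.
# It also computes parallax bubbles projected in pre-skinned space and sunlight-driven halftone dots with distance fade.
# Static switches select T2/T3 variants, hair handling and a second emissive layer; extra outputs expose masks.
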MaterialFunction /BRCosmetics/Characters/Player/Male/Medium/Bodies/M_MED_HerbHutch_Breath/Skins/Superlevels/Materials/MF_SL_S31
graph {nodes [
  n1 = GetMaterialAttributes(None=FunctionInput() {inputName=Material, inputType=FunctionInput_MaterialAttributes}) {attributeGetTypes [
    #1
    #2
    #3
    #4
    #5
    #6
    #7
   ]}
  n2 = SetMaterialAttributes!(
   MaterialAttributes=NamedRerouteDeclaration(None=@n1.MaterialAttributes) {name=MA_IN, nodeColor=<rgba 1 0 0.9609 1>}
   EmissiveColor=NamedRerouteUsage()
   BaseColor=NamedRerouteUsage()
   Metallic=NamedRerouteUsage()
   SubsurfaceColor=NamedRerouteUsage()
   Specular=NamedRerouteUsage()
   Roughness=NamedRerouteUsage()
   ShadingModel=ShadingModel()
  ) {attributeSetTypes [
    #1
    #2
    #4
    #5
    #6
    #7
    #8
   ]}
  n3 = SetMaterialAttributes!(
   MaterialAttributes=@n2
   BaseColor=Constant()
   EmissiveColor=NamedRerouteUsage()
   Roughness=ScalarParameter() {defaultValue=0.1, parameterName=SecondEmissive_Roughness}
   Metallic=ScalarParameter() {parameterName=SecondEmissive_Metallic}
   Specular=ScalarParameter() {defaultValue=0.5, parameterName=SecondEmissive_Specular}
  ) {attributeSetTypes [
    #2
    #1
    #7
    #4
    #6
   ]}
  n4 = NamedRerouteUsage()
  n5 = BlendMaterialAttributes(
   A=StaticSwitchParameter(True=BlendMaterialAttributes(A=@n2, B=@n3, Alpha=@n4), False=@n2) {group=S31 Superlevels, parameterName=UseSLFX2}
   B=NamedRerouteUsage()
   Alpha=NamedRerouteUsage()
  )
  n6 = NamedRerouteUsage()
  n7 = NamedRerouteUsage()
  n8 = Saturate(None=Multiply(
    A=NamedRerouteDeclaration(None=@n1.Roughness) {name=Roughness_IN, nodeColor=<rgba 1 0.4922 0 1>}
    B=ScalarParameter() {defaultValue=1, group=S31 Superlevels, parameterName=HairRoughnessScale}
   ))
  FunctionOutput(None=StaticSwitchParameter(
    True=SetMaterialAttributes!(MaterialAttributes=@n5, Metallic=@n6, Specular=@n7, Roughness=@n8) {attributeSetTypes [#4, #6, #7]}
    False=@n5
   ) {group=S31 Superlevels, parameterName=UseHair})
  n9 = ScalarParameter() {group=S31 Superlevels, parameterName=SLMask_G_Smoothstep_Min}
  n10 = ScalarParameter() {defaultValue=1, group=S31 Superlevels, parameterName=SLMask_G_Smoothstep_Max}
  n11 = TextureSampleParameter2D() {
   group=S31 Superlevels
   parameterName=SL_FX_Mask
   samplerType=SAMPLERTYPE_Masks
   sortPriority=0
   texture=/BRCosmetics/Characters/Player/Male/Medium/Bodies/M_MED_HerbHutch_Breath/Textures/T_HerbHutchBreath_Body_SL_FX
  }
  FunctionOutput(None=NamedRerouteDeclaration(None=SmoothStep(Min=@n9, Max=@n10, Value=@n11.G) {constMin=0.5}) {name=Green, nodeColor=<rgba 0 1 0.0966 1>}) {outputName=FX_Green, sortPriority=1}
  n12 = NamedRerouteDeclaration(None=SmoothStep(Value=MaterialFunctionCall(f3=@n11.RGB) {materialFunction=&AddComponents}.f3)) {name=Alpha, nodeColor=<rgba 0.9571 1 0.99 1>}
  FunctionOutput(None=@n12) {outputName=Alpha, sortPriority=2}
  n13 = Multiply(A=NamedRerouteUsage(), B=ScalarParameter() {defaultValue=0.8, group=S31 Superlevels, parameterName=EmissiveBasecolorRatio})
  LinearInterpolate(A=@n13, B=NamedRerouteUsage(), Alpha=NamedRerouteUsage()) {desc=Remove emissive on skin}
  n14 = VectorParameter() {group=S31 Superlevels, parameterName=Primary_Color, sortPriority=1, defaultValue=<rgba 0.4577 0.6354 0.1411 1>}
  n15 = NamedRerouteUsage()
  n16 = NamedRerouteUsage()
  n17 = NamedRerouteDeclaration(None=StaticSwitchParameter(True=NamedRerouteUsage(), False=StaticSwitchParameter(True=@n15, False=@n16) {group=S31 Superlevels, parameterName=UseT2}) {group=S31 Superlevels, parameterName=UseT3}) {name=PanningNoise, nodeColor=<rgba 0.5234 0 1 1>}
  n18 = SmoothStep(Value=NamedRerouteUsage()) {constMax=0.7}
  n19 = LinearInterpolate(A=Multiply(A=@n18) {constB=0.05}, B=Multiply(A=@n18) {constB=0.2}, Alpha=NamedRerouteUsage())
  n20 = Saturate(None=Subtract(A=Subtract(A=@n17, B=@n19), B=NamedRerouteUsage()))
  Multiply(A=@n14.RGB, B=Add(B=@n20) {constA=0.025})
  NamedRerouteDeclaration(None=SmoothStep(
    Min=ScalarParameter() {group=S31 Superlevels, parameterName=SLMask_R_Smoothstep_Min}
    Max=ScalarParameter() {defaultValue=1, group=S31 Superlevels, parameterName=SLMask_R_Smoothstep_Max}
    Value=@n11.R
   ) {constMin=0.5}) {name=Red, nodeColor=<rgba 1 0 0.0258 1>}
  n21 = Abs(None=ComponentMask(None=NamedRerouteUsage()) {b=true})
  n22 = ScalarParameter() {defaultValue=3, group=S31 Superlevels Bubbles, parameterName=TopDownMask_Power}
  NamedRerouteDeclaration(None=Saturate(None=Power(Base=@n21, Exp=@n22) {constExponent=3})) {name=TopDownMask, nodeColor=<rgba 0.125 0 1 1>}
  n23 = NamedRerouteUsage()
  NamedRerouteDeclaration(None=@n23) {name=Debug, nodeColor=<rgba 1 0 0.75 1>}
  NamedRerouteDeclaration(None=@n1.Metallic) {name=Metallic_IN, nodeColor=<rgba 0 0.25 1 1>}
  n24 = NamedRerouteUsage()
  NamedRerouteDeclaration(None=LinearInterpolate(A=NamedRerouteUsage(), B=ScalarParameter() {defaultValue=1, group=S31 Superlevels, parameterName=Quaternary_Metallic}, Alpha=@n24)) {name=Metallic_OUT, nodeColor=<rgba 0 1 0.8594 1>}
  n25 = LinearInterpolate(
   A=StaticSwitchParameter(True=Constant(), False=NamedRerouteUsage()) {group=S31 Superlevels, parameterName=UseHair}
   B=NamedRerouteUsage()
   Alpha=NamedRerouteUsage()
  )
  n26 = Saturate(None=Multiply(A=@n25, B=ScalarParameter() {defaultValue=1.75, group=S31 Superlevels, parameterName=DotIntensity}) {constB=1.75})
  n27 = NamedRerouteUsage()
  NamedRerouteDeclaration(None=LinearInterpolate(A=LinearInterpolate(A=@n25, B=@n26, Alpha=@n27) {desc=Add dots}, B=NamedRerouteUsage(), Alpha=NamedRerouteUsage())) {name=FinalColor, nodeColor=<rgba 1 0.8203 0 1>}
  NamedRerouteDeclaration(None=@n1.SubsurfaceColor) {name=SSColor_IN, nodeColor=<rgba 0 1 0.1797 1>}
  NamedRerouteDeclaration(None=LinearInterpolate(A=Constant(), B=NamedRerouteUsage(), Alpha=NamedRerouteUsage())) {name=SSColor_OUT, nodeColor=<rgba 1 0 0.1406 1>}
  n28 = MaterialFunctionCall(Float3=FunctionInput() {inputName=M, sortPriority=1}) {materialFunction=&BreakOutFloat3Components}
  NamedRerouteDeclaration(None=@n28.G) {name=GMap, nodeColor=<rgba 0.8281 0 1 1>}
  NamedRerouteDeclaration(None=@n11.B) {name=Blue, nodeColor=<rgba 0.049 0 1 1>}
  NamedRerouteDeclaration(None=@n28.R) {name=AO, nodeColor=<rgba 0 1 0.5078 1>}
  n29 = Fresnel() {baseReflectFraction=0, exponent=1}
  NamedRerouteDeclaration(None=OneMinus(None=NamedRerouteDeclaration(None=@n29) {name=Fresnel, nodeColor=<rgba 1 0.9609 0 1>})) {name=InvertedFresnel, nodeColor=<rgba 1 0.9375 0 1>}
  NamedRerouteDeclaration(None=@n1.Specular) {name=Specular_IN, nodeColor=<rgba 0 1 0.8125 1>}
  NamedRerouteDeclaration(None=LinearInterpolate(A=NamedRerouteUsage(), B=ScalarParameter() {defaultValue=0.9, group=S31 Superlevels, parameterName=Quaternary_Specular}, Alpha=@n24)) {name=Specular_OUT, nodeColor=<rgba 0.1719 0 1 1>}
  n30 = NamedRerouteUsage()
  n31 = Multiply(A=@n30, B=ScalarParameter() {defaultValue=1, group=S31 Superlevels, parameterName=RoughnessIn_Scale})
  n32 = LinearInterpolate(Alpha=@n30) {constA=0.25, constB=0.35}
  n33 = NamedRerouteUsage()
  NamedRerouteDeclaration(None=LinearInterpolate(A=Saturate(None=@n31), B=LinearInterpolate(B=@n32, Alpha=@n33) {constA=0.5, constB=0.5}, Alpha=@n24)) {name=Roughness_OUT, nodeColor=<rgba 0.1484 1 0 1>}
  n34 = NamedRerouteDeclaration(None=Saturate(None=Subtract(A=@n28.B, B=@n12))) {name=SkinMask, nodeColor=<rgba 0 1 0.9531 1>}
  NamedRerouteDeclaration(None=Subtract(A=OneMinus(None=@n34), B=@n12)) {name=NonSLMask, nodeColor=<rgba 1 0.2109 0 1>}
  n35 = TextureSampleParameter2D() {
   group=S31 Superlevels
   parameterName=SL_FX_Mask2
   samplerType=SAMPLERTYPE_Masks
   sortPriority=0
   texture=/BRCosmetics/Characters/Player/Male/Medium/Bodies/M_MED_HerbHutch_Breath/Textures/T_HerbHutchBreath_Body_SL_FX2
  }
  NamedRerouteDeclaration(None=@n35.R) {name=Red2, nodeColor=<rgba 0.7578 0 1 1>}
  n36 = StaticSwitchParameter(True=Constant(), False=NamedRerouteUsage()) {group=S31 Superlevels, parameterName=UseHair}
  NamedRerouteDeclaration(None=Multiply(
    A=LinearInterpolate(A=@n13, Alpha=@n36) {constB=0}
    B=ScalarParameter() {defaultValue=2, group=S31 Superlevels, parameterName=GlobalEmissiveBoost}
   )) {name=Emissive_OUT, nodeColor=<rgba 0 0.8359 1 1>}
  n37 = StaticSwitchParameter(
   True=VectorParameter() {group=S31 Superlevels, parameterName=SecondEmissiveColor_T3, defaultValue=<rgba 0.974 0.1333 0 1>}.RGB
   False=StaticSwitchParameter(
    True=VectorParameter() {group=S31 Superlevels, parameterName=SecondEmissiveColor_T2, defaultValue=<rgba 0 0.3071 0.974 1>}.RGB
    False=VectorParameter() {group=S31 Superlevels, parameterName=SecondEmissiveColor_T1, defaultValue=<rgba 0.2903 0.599 0 1>}.RGB
   ) {group=S31 Superlevels, parameterName=UseT2}
  ) {group=S31 Superlevels, parameterName=UseT3}
  n38 = NamedRerouteUsage()
  NamedRerouteDeclaration(None=LinearInterpolate(A=Multiply(A=@n37, B=@n38), B=NamedRerouteUsage(), Alpha=NamedRerouteUsage())) {name=FinalColor2, nodeColor=<rgba 0 0.3437 1 1>}
  NamedRerouteDeclaration(None=ScalarParameter() {defaultValue=25, group=S31 Superlevels Bubbles, parameterName=BubbleSize_A}) {name=BubbleSize_A, nodeColor=<rgba 0.0312 0 1 1>}
  NamedRerouteDeclaration(None=ScalarParameter() {defaultValue=30, group=S31 Superlevels Bubbles, parameterName=BubbleSize_A_1}) {name=BubbleSize_A_1, nodeColor=<rgba 0.0312 0 1 1>}
  NamedRerouteDeclaration(None=ScalarParameter() {defaultValue=15, group=S31 Superlevels Bubbles, parameterName=BubbleSize_A_2}) {name=BubbleSize_A_2, nodeColor=<rgba 0.0312 0 1 1>}
  n39 = NamedRerouteUsage()
  n40 = CameraPositionWS()
  n41 = Normalize(VectorInput=Subtract(A=@n39, B=TransformPosition(None=@n40) {transformSourceType=TRANSFORMPOSSOURCE_World}))
  n42 = Multiply(A=@n41, B=ScalarParameter() {defaultValue=20, group=S31 Superlevels Bubbles, parameterName=Bubble_ParallaxDepth_Layer3}) {constB=15}
  n43 = ComponentMask(None=@n39) {b=true, g=true, r=true}
  n44 = Add(A=@n43) {constB=166}
  n45 = QualitySwitch(Default=Add(A=Add(A=@n43, B=@n42)) {constB=166}, Low=@n44)
  NamedRerouteDeclaration(None=Add(
    A=ShadingPathSwitch(Default=@n45, Deferred=@n45, Forward=@n44, Mobile=@n44)
    B=VectorParameter() {group=S31 Superlevels Bubbles, parameterName=Bubble Position_3_Offset, defaultValue=<rgba 0 5 0 1>}.RGB
   )) {name=BubbleWorldPosParallax_3, nodeColor=<rgba 1 0 0.0469 1>}
  n46 = ScalarParameter() {defaultValue=15, group=S31 Superlevels Bubbles, parameterName=Bubble_ParallaxDepth_Layer1}
  n47 = QualitySwitch(Default=Add(A=@n43, B=Multiply(A=@n41, B=@n46) {constB=12}), Low=@n43)
  NamedRerouteDeclaration(None=Add(
    A=ShadingPathSwitch(Default=@n47, Deferred=@n47, Forward=@n43, Mobile=@n43)
    B=VectorParameter() {group=S31 Superlevels Bubbles, parameterName=Bubble Position_1_Offset, defaultValue=<rgba 0 5 0 1>}.RGB
   )) {name=BubbleWorldPosParallax_1, nodeColor=<rgba 0.0781 0 1 1>}
  n48 = VectorParameter() {group=S31 Superlevels Bubbles, parameterName=Bubble_DistortAmplitude_1_Bubbly, defaultValue=<rgba 5 2 0.1 1>}
  n49 = ComponentMask(None=NamedRerouteUsage()) {b=true}
  n50 = Multiply(A=Time()) {constB=0.5}
  n51 = Sine(None=Add(A=Multiply(A=@n49) {constB=0.005}, B=Frac(None=@n50)))
  NamedRerouteDeclaration(None=Multiply(A=@n48.RGB, B=@n51)) {name=BubbleDistort_1, nodeColor=<rgba 1 0 0.3281 1>}
  n52 = VectorParameter() {group=S31 Superlevels Bubbles, parameterName=Bubble_DistortAmplitude_2_Bubbly, defaultValue=<rgba -2 2 1 1>}
  NamedRerouteDeclaration(None=Multiply(A=@n51, B=@n52.RGB)) {name=BubbleDistort_2, nodeColor=<rgba 1 0 0.3281 1>}
  n53 = Multiply(A=@n41, B=ScalarParameter() {defaultValue=14, group=S31 Superlevels Bubbles, parameterName=Bubble_ParallaxDepth_Layer2}) {constB=15}
  n54 = Add(A=@n43) {constB=20}
  n55 = QualitySwitch(Default=Add(A=Add(A=@n43, B=@n53)) {constB=20}, Low=@n54)
  NamedRerouteDeclaration(None=Add(
    A=ShadingPathSwitch(Default=@n55, Deferred=@n55, Forward=@n54, Mobile=@n54)
    B=VectorParameter() {group=S31 Superlevels Bubbles, parameterName=Bubble Position_2_Offset, defaultValue=<rgba 0 5 0 1>}.RGB
   )) {name=BubbleWorldPosParallax_2, nodeColor=<rgba 0.0781 0 1 1>}
  NamedRerouteDeclaration(None=MaterialFunctionCall() {materialFunction=&BoundingBoxBased_0-1_UVW}.B) {name=BBox, nodeColor=<rgba 0.9922 1 0 1>}
  NamedRerouteDeclaration(None=@n1.EmissiveColor) {name=Emissive_IN, nodeColor=<rgba 0.8047 1 0 1>}
  n56 = PreSkinnedNormal()
  NamedRerouteDeclaration(None=ShaderStageSwitch(PixelShader=VertexInterpolator(VS=@n56).PS, VertexShader=@n56)) {name=PS-LocalNrm, nodeColor=<rgba 0 0.1867 0.3854 1>}
  n57 = TextureObject() {
   samplerType=SAMPLERTYPE_LinearColor
   texture=/BRCosmetics/Characters/Player/Male/Medium/Bodies/M_MED_HerbHutch_Breath/Skins/Superlevels/Textures/T_SL_S31_Bubbles
  }
  n58 = ScalarParameter() {defaultValue=80, parameterName=Bubbles_Tiling}
  n59 = ScalarParameter() {defaultValue=0.23, group=S31 Superlevels Bubbles, parameterName=Bubbles_Speed}
  n60 = MaterialFunctionCall!(
   TextureObject=@n57
   TextureSize=@n58
   WorldPosition=NamedRerouteUsage()
   In=Multiply(A=@n59, B=Constant2Vector() {g=1, r=0.654})
   World Space Normal=NamedRerouteUsage()
   ProjectionTransitionContrast=ScalarParameter() {group=S31 Superlevels Bubbles, parameterName=Bubble_Blend_SharpnessMasks_2, sliderMax=1}
  ) {materialFunction=&/Preskinned_WP_Texture_ScalePan_XY}
  NamedRerouteDeclaration(None=ComponentMask(None=@n60.XY Texture) {b=true}) {name=BackgroundBubbles, nodeColor=<rgba 0 1 0.7891 1>}
  n61 = NamedRerouteUsage()
  n62 = SmoothStep(Value=SphereMask(A=ComponentMask(None=@n61) {b=true}, B=Constant()) {attenuationRadius=165, hardnessPercent=0.8}) {constMax=0.25}
  n63 = ScalarParameter() {group=S31 Superlevels Bubbles, parameterName=Bubble_Blend_SharpnessMasks, sliderMax=1}
  n64 = MaterialFunctionCall(
   TextureObject=@n57
   TextureSize=@n58
   WorldPosition=Add(A=NamedRerouteUsage(), B=NamedRerouteUsage())
   In=Multiply(A=@n59, B=Constant2Vector() {g=1})
   World Space Normal=NamedRerouteUsage()
   ProjectionTransitionContrast=@n63
  ) {materialFunction=&Preskinned_WP_Texture_ScalePan_XY}
  n65 = ComponentMask(None=@n64.XY Texture) {r=true}
  n66 = NamedRerouteUsage()
  MaterialExpressionTime_3 = Time()
  MaterialExpressionScalarParameter_55 = ScalarParameter() {defaultValue=1.5, group=S31 Superlevels Bubbles, parameterName=BubbleScalingSpeed2}
  MaterialExpressionConstant_10 = Constant() {r=0.333}
  n67 = Custom!(time=@MaterialExpressionTime_3, speed=@MaterialExpressionScalarParameter_55, offset=@MaterialExpressionConstant_10) {
   code="return (sin(time*speed+offset)+1)/2;"
   description=TimeSineOffset
   inputs [
    {
     inputName=time
     input {expression=&self:MaterialExpressionTime_3}
    }
    {
     inputName=speed
     input {expression=&self:MaterialExpressionScalarParameter_55}
    }
    {
     inputName=offset
     input {expression=&self:MaterialExpressionConstant_10}
    }
   ]
  }
  n68 = LinearInterpolate(Alpha=@n67) {constA=0.3, constB=0.7}
  n69 = Add(A=@n68)
  n70 = Saturate(None=Subtract(A=ComponentMask(None=@n64.XY Texture) {g=true}, B=@n66))
  n71 = Saturate(None=Subtract(A=ComponentMask(None=@n60.XY Texture) {g=true}, B=NamedRerouteUsage()) {constB=0})
  n72 = MaterialFunctionCall(
   TextureObject=@n57
   TextureSize=@n58
   WorldPosition=Add(A=NamedRerouteUsage(), B=NamedRerouteUsage())
   In=Multiply(A=@n59, B=Constant2Vector() {g=0.85, r=-0.5})
   World Space Normal=NamedRerouteUsage()
   ProjectionTransitionContrast=@n63
  ) {materialFunction=&Preskinned_WP_Texture_ScalePan_XY}
  n73 = Saturate(None=Subtract(A=ComponentMask(None=@n72.XY Texture) {b=true}, B=NamedRerouteUsage()) {constB=0})
  n74 = Multiply(
   A=Saturate(None=Subtract(A=@n65, B=@n66) {constB=0})
   B=Max(A=SmoothStep(Min=@n68, Max=@n69, Value=@n70) {constMax=0.5, constMin=0.4}, B=Max(A=@n71, B=@n73))
  )
  NamedRerouteDeclaration(None=Multiply(A=NamedRerouteDeclaration(None=@n62) {name=FacialMask, nodeColor=<rgba 0.5 0 1 1>}, B=SmoothStep(Value=@n74) {constMax=0.5, constMin=0.49})) {name=Msk_Bubbles, nodeColor=<rgba 1 0 0.6328 1>}
  n75 = ComponentMask(None=NamedRerouteUsage()) {b=true, r=true}
  n76 = Divide(A=@n75, B=Constant2Vector() {g=128, r=100}) {constB=128}
  n77 = ComponentMask(None=TextureSample(UVs=Panner(Coordinate=@n76) {speedX=0.007, speedY=0.085}, Tex=NamedRerouteUsage()) {samplerSource=SSM_Wrap_WorldGroupSettings}.RGB) {r=true}
  n78 = TextureSample(UVs=Add(A=Multiply(A=@n77) {constB=0.1}, B=Panner(Coordinate=@n76) {speedX=-0.016, speedY=0.016}), Tex=NamedRerouteUsage()) {samplerSource=SSM_Wrap_WorldGroupSettings, samplerType=SAMPLERTYPE_Grayscale}
  NamedRerouteDeclaration(None=Saturate(None=Power(Base=@n78.R) {constExponent=2.5})) {name=PrimaryPanningNoise1, nodeColor=<rgba 0.2188 0 1 1>}
  n79 = CollectionParameter() {
   collection=&FortniteMaterialParameters
   parameterName=SunlightDirectionalVector
  }
  n80 = DotProduct(A=VertexNormalWS(), B=Normalize(VectorInput=@n79))
  n81 = Add(A=@n80) {constB=0.01}
  n82 = NamedRerouteUsage()
  n83 = ScalarParameter() {defaultValue=50, group=01_InvertedHull_Outlines, parameterName=Fade_Offset, sliderMax=100, sortPriority=5}
  n84 = ScalarParameter() {defaultValue=200, group=01_InvertedHull_Outlines, parameterName=Fade_Distance, sliderMax=1500, sliderMin=500, sortPriority=5}
  NamedRerouteDeclaration(None=LinearInterpolate(
    A=SmoothStep(Min=@n81, Max=@n80, Value=@n82)
    B=@n80
    Alpha=MaterialFunctionCall!(Fade Offset=@n83, Fade Distance=@n84) {materialFunction=/Game/Global/Materials/Character_Parent_2023/Anime_Parent_2023/MaterialFunctions/MFC_DistanceFade_FOV_compensation}.Result
   ) {constB=0, desc=Fade dots in distance}) {name=Dots, nodeColor=<rgba 1 0.0703 0 1>}
  n85 = MaterialFunctionCall() {materialFunction=&CameraDirectionVector}
  n86 = Multiply(A=Constant3Vector() {constant=<rgba 1 1 0 0>}, B=@n85.Result)
  n87 = Add(A=Multiply(B=Normalize(VectorInput=@n86)) {constA=0.25}, B=@n85.Result)
  n88 = Multiply(A=@n85.Result) {constB=50}
  n89 = Add(A=ActorPositionWS(), B=Constant3Vector() {constant=<rgba 0 0 90 1>})
  n90 = Constant() {r=64}
  n91 = Constant() {r=128}
  n92 = InverseLinearInterpolate(
   A=Constant() {r=45}
   B=Constant() {r=90}
   Value=MaterialFunctionCall() {materialFunction=/Engine/Functions/Engine_MaterialFunctions02/Utility/FOV}.FOV angle in degrees
  )
  NamedRerouteDeclaration(None=MaterialFunctionCall(Plane Normal Axis=Normalize(VectorInput=@n87), Plane Center=Add(A=@n88, B=@n89), UV Scale=LinearInterpolate(A=@n90, B=@n91, Alpha=@n92)) {materialFunction=&VirtualPlaneCoordinates}.UVs) {name=ObjectAlighedScreenPosition, nodeColor=<rgba 1 0.867 0 1>}
  n93 = Subtract(A=Distance(A=ActorPositionWS(), B=CameraPositionWS())) {constB=100}
  NamedRerouteDeclaration(None=Saturate(None=Divide(A=@n93) {constB=400})) {name=DistanceBlending, nodeColor=<rgba 0.4766 1 0 1>}
  n94 = ScalarParameter() {defaultValue=120, group=01_InvertedHull_Outlines, parameterName=DotsAmount, sliderMax=512, sortPriority=20}
  n95 = Multiply(A=NamedRerouteUsage(), B=LinearInterpolate(A=@n94, B=Multiply(A=@n94) {constB=0.5}, Alpha=NamedRerouteUsage()))
  n96 = Constant() {r=1}
  n97 = Add(A=Frac(None=Rotator(Coordinate=@n95, Time=@n96) {speed=0.75})) {constB=-0.5}
  n98 = Constant()
  NamedRerouteDeclaration(None=Add(A=Distance(A=@n97, B=@n98)) {constB=-0.1, desc=to get full transition}) {name=DotsGradient, nodeColor=<rgba 1 0.0938 0 1>}
  n99 = TextureObject() {
   samplerType=SAMPLERTYPE_LinearColor
   texture=/BRCosmetics/Characters/Player/Male/Medium/Bodies/M_MED_HerbHutch_Breath/Skins/Superlevels/Textures/T_SL_S31_PerlinRGB
  }
  NamedRerouteDeclaration(None=@n99) {name=PrimaryNoiseTex, nodeColor=<rgba 0.7344 0 1 1>}
  n100 = VectorParameter() {group=S31 Superlevels, parameterName=Primary_Color_2, sortPriority=1, defaultValue=<rgba 0.025 0.025 0.025 1>}
  n101 = LinearInterpolate(A=LinearInterpolate(A=@n100.RGB, B=@n14.RGB, Alpha=@n20), Alpha=NamedRerouteUsage()) {constB=0}
  n102 = NamedRerouteUsage()
  n103 = ScalarParameter() {defaultValue=1, group=S31 Superlevels, parameterName=Primary_FresnelDarken}
  NamedRerouteDeclaration(None=Multiply(A=Multiply(A=@n101, B=@n23), B=LinearInterpolate(B=@n102, Alpha=@n103) {constA=1})) {name=PrimaryColor, nodeColor=<rgba 0 1 0.4609 1>}
  n104 = VectorParameter() {group=S31 Superlevels, parameterName=Secondary_GradientColor_2, sortPriority=2, defaultValue=<rgba 0 0 0 1>}
  n105 = VectorParameter() {group=S31 Superlevels, parameterName=Secondary_GradientColor_1, sortPriority=2, defaultValue=<rgba 0.3996 0.4115 0.0411 1>}
  n106 = Divide(A=ComponentMask(None=NamedRerouteUsage()) {b=true}) {constB=150}
  n107 = Add(A=@n106, B=Multiply(A=Time(), B=ScalarParameter() {defaultValue=-0.2, group=S31 Superlevels, parameterName=Secondary_GradientPanningSpeed}))
  n108 = Multiply(
   A=LinearInterpolate(
    A=ScalarParameter() {defaultValue=0.4, group=S31 Superlevels, parameterName=Secondary_GradientMultiMin}
    B=ScalarParameter() {defaultValue=0.75, group=S31 Superlevels, parameterName=Secondary_GradientMultiMax}
    Alpha=Saturate(None=@n106)
   ) {constA=0.4, constB=0.75}
   B=ConstantBiasScale(None=Sine(None=@n107))
  )
  n109 = NamedRerouteUsage()
  NamedRerouteDeclaration(None=LinearInterpolate(A=@n104.RGB, B=@n105.RGB, Alpha=Multiply(A=@n108, B=@n109))) {name=SecondaryColor, nodeColor=<rgba 1 0.1406 0 1>}
  n110 = VectorParameter() {group=S31 Superlevels, parameterName=Tertiary_EdgeColor_1, sortPriority=3, defaultValue=<rgba 0.5355 1 0.12 1>}
  n111 = VectorParameter() {group=S31 Superlevels, parameterName=Tertiary_EdgeColor_2, sortPriority=3, defaultValue=<rgba 0.4487 0.9 0.045 1>}
  n112 = NamedRerouteUsage()
  NamedRerouteDeclaration(None=LinearInterpolate(A=@n110.RGB, B=@n111.RGB, Alpha=Saturate(None=@n112))) {name=TertiaryColor, nodeColor=<rgba 0 1 0.2031 1>}
  n113 = TextureObject() {
   samplerType=SAMPLERTYPE_LinearColor
   texture=/BRCosmetics/Characters/Player/Male/Medium/Bodies/M_MED_HerbHutch_Breath/Skins/Superlevels/Textures/T_SL_S31_PerlinRGB_T2
  }
  n114 = TextureSample(UVs=Panner(Coordinate=Divide(A=@n75) {constB=80}) {speedX=-0.034, speedY=0.145}, Tex=@n113) {
   samplerSource=SSM_Wrap_WorldGroupSettings
   samplerType=SAMPLERTYPE_LinearColor
   texture=&T_SL_S31_PerlinRGB
  }
  n115 = TextureSample(UVs=Add(A=Multiply(A=@n114.B) {constB=0.15}, B=Divide(A=@n75) {constB=128}), Tex=@n113) {
   samplerSource=SSM_Wrap_WorldGroupSettings
   samplerType=SAMPLERTYPE_LinearColor
   texture=&T_SL_S31_PerlinRGB
  }
  n116 = Multiply(A=@n115.G, B=@n114.G)
  n117 = Power(Base=Add(A=SmoothStep(Value=@n115.G) {constMin=0.2}, B=Multiply(A=@n116) {constB=2})) {constExponent=1.7}
  NamedRerouteDeclaration(None=Add(A=Multiply(A=@n117) {constB=3.5}) {constB=0.1}) {name=PrimaryPanningNoise2, nodeColor=<rgba 0.4766 0 1 1>}
  n118 = Divide(A=@n75, B=Constant2Vector() {g=100, r=50}) {constB=100}
  n119 = TextureSample(UVs=Panner(Coordinate=@n118) {bFractionalPart=true, speedY=-0.31}, Tex=NamedRerouteUsage()) {
   samplerSource=SSM_Wrap_WorldGroupSettings
   samplerType=SAMPLERTYPE_LinearColor
   texture=&T_SL_S31_PerlinRGB
  }
  n120 = ComponentMask(None=@n119.RGB) {b=true, r=true}
  n121 = TextureSample(UVs=Add(A=Panner(Coordinate=@n118) {bFractionalPart=true, speedY=-0.1245}, B=Multiply(A=@n120) {constB=0.2}), Tex=NamedRerouteUsage()) {
   samplerSource=SSM_Wrap_WorldGroupSettings
   samplerType=SAMPLERTYPE_LinearColor
   texture=&T_SL_S31_PerlinRGB
  }
  n122 = Multiply(A=@n121.R, B=@n121.R)
  NamedRerouteDeclaration(None=Multiply(A=Power(Base=@n122) {constExponent=2.2}) {constB=7}) {name=PrimaryPanningNoise3, nodeColor=<rgba 0 1 0.6016 1>}
  NamedRerouteDeclaration(None=@n121.B) {name=FlamingNoise, nodeColor=<rgba 0 1 0.3906 1>}
  n123 = ScalarParameter() {defaultValue=1.25, group=S31 Superlevels, parameterName=HairFresnel}
  n124 = Constant() {r=2}
  n125 = SmoothStep(Value=Fresnel(ExponentIn=StaticSwitchParameter(True=@n123, False=@n124) {group=S31 Superlevels, parameterName=UseHair}) {baseReflectFraction=0, exponent=2}) {constMin=0.3, desc=Tighten fresnel}
  n126 = NamedRerouteUsage()
  NamedRerouteDeclaration(None=Saturate(None=Add(A=@n125, B=@n126))) {name=TertiaryMask, nodeColor=<rgba 1 0 0.1172 1>}
  NamedRerouteDeclaration(None=@n1.BaseColor) {name=BaseColor_IN, nodeColor=<rgba 1 0.6328 0 1>}
  n127 = NamedRerouteUsage()
  n128 = ScalarParameter() {defaultValue=0.5, group=S31 Superlevels, parameterName=AOIntensity}
  NamedRerouteDeclaration(None=Multiply(A=NamedRerouteUsage(), B=LinearInterpolate(A=@n127, Alpha=@n128))) {name=SecondaryMask, nodeColor=<rgba 0 0.4375 1 1>}
  n129 = NamedRerouteUsage()
  n130 = OneMinus(None=ScalarParameter() {defaultValue=0.8, group=S31 Superlevels, parameterName=EmissiveBasecolorRatio})
  n131 = VectorParameter() {group=S31 Superlevels, parameterName=Hair_Color, defaultValue=<rgba 0.02 0.02 0.02 1>}
  n132 = VectorParameter() {group=S31 Superlevels, parameterName=Quaternary_Color, defaultValue=<rgba 0.4142 0.4427 0.2398 1>}
  n133 = ScalarParameter() {defaultValue=1, group=S31 Superlevels, parameterName=SL_ThinFilmScale, sortPriority=5}
  n134 = NamedRerouteUsage()
  n135 = Power(Base=@n134, Exp=ScalarParameter() {defaultValue=1, group=S31 Superlevels, parameterName=SL_ThinFilmWarp, sortPriority=4})
  n136 = TextureSampleParameter2D(UVs=Add(A=Multiply(A=@n133, B=@n135), B=ScalarParameter() {group=S31 Superlevels, parameterName=SL_ThinFilmOffset})) {
   desc=Optimize
   group=S31 Superlevels
   parameterName=SL_Thin Film Texture
   sortPriority=1
   texture=/Game/Global/Textures/Default/Thinfilm/T_ThinFilm_Spectrum_COLOR
  }
  n137 = ScalarParameter() {defaultValue=0.2, group=S31 Superlevels, parameterName=SL_ThinFilmIntensity, sortPriority=2}
  n138 = ScalarParameter() {defaultValue=0.2, group=S31 Superlevels, parameterName=SL_ThinFilmExponent, sortPriority=3}
  n139 = LinearInterpolate(A=@n132.RGB, B=Add(A=@n132.RGB, B=Multiply(A=@n136.RGB, B=@n137)), Alpha=Saturate(None=Power(Base=@n134, Exp=@n138)))
  n140 = LinearInterpolate(
   A=LinearInterpolate(
    A=Multiply(A=@n129, B=@n130)
    B=StaticSwitchParameter(True=@n131.RGB, False=@n139) {group=S31 Superlevels, parameterName=UseHair}
    Alpha=NamedRerouteUsage()
   ) {constB=0}
   Alpha=NamedRerouteUsage()
  ) {constB=0}
  NamedRerouteDeclaration(None=LinearInterpolate(A=Clamp(None=@n140) {minDefault=0.01}, B=NamedRerouteUsage(), Alpha=NamedRerouteUsage())) {name=BaseColor_OUT, nodeColor=<rgba 0.3125 0 1 1>}
  n141 = LocalPosition() {localOrigin=InstancePreSkinning}
  NamedRerouteDeclaration(None=ShaderStageSwitch(PixelShader=VertexInterpolator(VS=@n141.XYZ).PS, VertexShader=@n141.XYZ)) {name=PS-LocalPos, nodeColor=<rgba 0 1 1 1>}
  NamedRerouteDeclaration(None=@n1.Normal) {name=Normal_IN, nodeColor=<rgba 0.5938 0 1 1>}
  NamedRerouteUsage()
  ScalarParameter() {defaultValue=0.8, parameterName=hard}
  ScalarParameter() {defaultValue=165, parameterName=Radius}
  SmoothStep(Value=@n35.R) {constMax=0.5}
  Transform(None=NamedRerouteUsage()) {transformType=TRANSFORM_Local}
 ]}
