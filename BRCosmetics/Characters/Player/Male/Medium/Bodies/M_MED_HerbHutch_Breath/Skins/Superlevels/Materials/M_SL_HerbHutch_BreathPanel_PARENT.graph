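# Parent skinned-character material for the HerbHutch Breath Season 31 Superlevels skin, layering superlevel effects with Hatchback comic shading.
# Comic base color, with specular, roughness and metallic all wired to the optional SRM texture's red channel, feeds MF_SL_S31 alongside a mask.
# The mask's blue channel blends that result back toward comic shading, while a UseEyes switch substitutes an MF_SL_Eyes blend.
# The shared 2023 character PostFX function finishes the output.
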Material /BRCosmetics/Characters/Player/Male/Medium/Bodies/M_MED_HerbHutch_Breath/Skins/Superlevels/Materials/M_SL_HerbHutch_BreathPanel_PARENT
graph {
 properties {
  bAutomaticallySetUsageInEditor=false
  bUseMaterialAttributes=true
  bUsedWithMeshDeformer=true
  bUsedWithMorphTargets=true
  bUsedWithSkeletalMesh=true
  bUsedWithStaticLighting=true
  pixelDepthOffsetMode=PDOM_Legacy
  shadingModel=MSM_FromMaterialExpression
  usedShadingModels=Default Lit
 }
 nodes [
  Composite() {subgraphName=Construct Base Inputs}
  n1 = MaterialFunctionCall!(CustomUVs=MaterialFunctionCall() {materialFunction=/Game/Characters/Shared/FNEye/Materials/MF_EyeUVs}.EyeUVs) {materialFunction=/BRCosmetics/Characters/Player/Male/Medium/Bodies/M_MED_DriedSilk/Materials/MF_Hatchback_ComicShading}
  n2 = GetMaterialAttributes(None=@n1.Result) {attributeGetTypes [#1]}
  n3 = TextureSampleParameter2D() {
   group=S31 Superlevels
   parameterName=SRM
   samplerType=SAMPLERTYPE_Masks
   texture=/BRCosmetics/Characters/Player/Male/Medium/Bodies/M_MED_HerbHutch_Breath/Textures/T_HerbHutch_Breath_Body_S
  }
  n4 = MaterialFunctionCall(Float3=StaticSwitchParameter(True=@n3.RGB, False=Constant3Vector()) {group=S31 Superlevels, parameterName=UseSRM}) {materialFunction=&BreakOutFloat3Components}
  n5 = SetMaterialAttributes!(MaterialAttributes=@n2.MaterialAttributes, Base Color=@n2.MaterialAttributes, Specular=@n4.R, Roughness=@n4.R, Metallic=@n4.R) {attributeSetTypes [
    #2
    #3
    #4
    #5
   ]}
  n6 = TextureSampleParameter2D() {
   group=S31 Superlevels
   parameterName=Mask
   texture=/BRCosmetics/Characters/Player/Female/Medium/Bodies/F_MED_PillowMill_Mastery/Textures/T_PillowMill_Mastery_Body_M
  }
  n7 = MaterialFunctionCall!(Material=@n5, M=@n6.RGB) {materialFunction=&/MF_SL_S31}
  n8 = MaterialFunctionCall() {materialFunction=&/MF_SL_Eyes}
  n9 = Constant()
  MaterialExpressionReroute_0 = Reroute(None=@n9)
  n10 = PreSkinnedPosition()
  n11 = ShaderStageSwitch(PixelShader=VertexInterpolator(VS=@n10).PS, VertexShader=@n10)
  MaterialExpressionReroute_1 = Reroute(None=@n11)
  n12 = MaterialFunctionCall!(Material) {materialFunction=/Game/Packages/Fortress_SharedMaterials/Base_Material_Functions/CharacterBrightnessandScale_Balance}
  MaterialExpressionReroute_2 = Reroute(None=@n12.Distance Blend)
  n13 = Fresnel() {baseReflectFraction=0, exponent=1}
  MaterialExpressionReroute_3 = Reroute(None=@n13)
  MaterialFunctionCall(
   Material=StaticSwitchParameter(
    True=BlendMaterialAttributes(A=@n1.Result, B=@n8.Result, Alpha=@n8.Result)
    False=BlendMaterialAttributes(A=@n7.Result, B=@n1.Result, Alpha=@n6.B)
   ) {group=S31 Superlevels, parameterName=UseEyes}
   M=@MaterialExpressionReroute_0.M
   Pre Skinned Local Position=@MaterialExpressionReroute_1.Pre Skinned Local Position
   Distance Blend=@MaterialExpressionReroute_2.Distance Blend
   Fresnel=@MaterialExpressionReroute_3.Fresnel
  ) {materialFunction=&MF_CharacterParent_2023_PostFX}
  PinBase()
  PinBase!(M=@n9, Pre Skinned Local Position=@n11, Distance Blend=@n12.Distance Blend, Fresnel=@n13) {reroutePins [
    M {expression=&self:MaterialExpressionReroute_0}
    Pre Skinned Local Position {expression=&self:MaterialExpressionReroute_1}
    Distance Blend {expression=&self:MaterialExpressionReroute_2}
    Fresnel {expression=&self:MaterialExpressionReroute_3}
   ]}
 ]
}
