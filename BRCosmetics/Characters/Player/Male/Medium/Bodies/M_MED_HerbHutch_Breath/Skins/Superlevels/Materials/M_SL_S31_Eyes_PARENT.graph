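# M_SL_S31_Eyes_PARENT is a parent eye material for skeletal character meshes, stored with the HerbHutch Breath body's Superlevels skins.
# It outputs material attributes from the MF_SL_Eyes function, passed through MF_TechArtOverrides before reaching the final output.
# All eye shading logic lives in those functions, and the shading model comes from the material expression, resolving to Default Lit, with mesh deformer usage enabled.
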Material /BRCosmetics/Characters/Player/Male/Medium/Bodies/M_MED_HerbHutch_Breath/Skins/Superlevels/Materials/M_SL_S31_Eyes_PARENT
graph {
 properties {
  bUseMaterialAttributes=true
  bUsedWithMeshDeformer=true
  bUsedWithSkeletalMesh=true
  pixelDepthOffsetMode=PDOM_Legacy
  shadingModel=MSM_FromMaterialExpression
  usedShadingModels=Default Lit
 }
 nodes [MaterialFunctionCall(In=MaterialFunctionCall() {materialFunction=&/MF_SL_Eyes}.Result) {materialFunction=&MF_TechArtOverrides}]
}
