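# Parent character material for the S31 Superlevels skins, supporting skeletal meshes and several shading models via material attributes.
# The shared 2023 character PreFX function feeds the MF_SL_S31 effect and an MF_NewDiamond_SL layer.
# The static switch UsePiperShelfDiamond either blends the diamond layer over the S31 result or passes that result through alone.
# The shared PostFX function receives the chosen result along with PreFX data to finish the material.
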Material /BRCosmetics/Characters/Player/Male/Medium/Bodies/M_MED_HerbHutch_Breath/Skins/Superlevels/Materials/M_SL_S31_PARENT
graph {
 properties {
  bUseMaterialAttributes=true
  bUsedWithMeshDeformer=true
  bUsedWithMorphTargets=true
  bUsedWithSkeletalMesh=true
  pixelDepthOffsetMode=PDOM_Legacy
  shadingModel=MSM_FromMaterialExpression
  usedShadingModels=Default Lit | Subsurface | Clear Coat | Subsurface Profile | Cloth
 }
 nodes [
  n1 = MaterialFunctionCall() {materialFunction=&MF_CharacterParent_2023_PreFX}
  n2 = MaterialFunctionCall!(Material=@n1.Result, M=@n1.Result) {materialFunction=&/MF_SL_S31}
  n3 = MaterialFunctionCall!(Material=@n1.Result, TechArtMask=@n1.Result) {materialFunction=&/MF_NewDiamond_SL}
  MaterialFunctionCall(
   Material=StaticSwitchParameter(True=BlendMaterialAttributes(A=@n2.Result, B=@n3.Material, Alpha=@n3.Material), False=@n2.Result) {group=S31 Superlevels, parameterName=UsePiperShelfDiamond}
   M=@n1.M
   Pre Skinned Local Position=@n1.M
   Distance Blend=@n1.M
   Fresnel=@n1.M
  ) {materialFunction=&MF_CharacterParent_2023_PostFX}
 ]
}
